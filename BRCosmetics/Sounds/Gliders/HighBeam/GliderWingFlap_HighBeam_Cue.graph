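# Sound cue for the HighBeam glider's wing flap sound effect.
# A Random node picks among six equally weighted flap recordings and feeds a Modulator that varies each play, with the pitch floor raised to unshifted.
# It uses the close in-air player glider attenuation and is assigned to the glider landing sound class.
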SoundCue /BRCosmetics/Sounds/Gliders/HighBeam/GliderWingFlap_HighBeam_Cue
root=Modulator_0
nodes [
 {id=Modulator_0, children [Random_0], properties {pitchMin=1}}
 {
  id=Random_0
  children [WavePlayer_6, WavePlayer_7, WavePlayer_8, WavePlayer_9, WavePlayer_10, WavePlayer_11]
  properties {weights [1,1,1,1,1,1]}
 }
 {id=WavePlayer_10, properties {soundWaveAssetPtr=&/IP_GliderWingFlaps_HighBeam_05}}
 {id=WavePlayer_11, properties {soundWaveAssetPtr=&/IP_GliderWingFlaps_HighBeam_06}}
 {id=WavePlayer_6, properties {soundWaveAssetPtr=&/IP_GliderWingFlaps_HighBeam_01}}
 {id=WavePlayer_7, properties {soundWaveAssetPtr=&/IP_GliderWingFlaps_HighBeam_02}}
 {id=WavePlayer_8, properties {soundWaveAssetPtr=&/IP_GliderWingFlaps_HighBeam_03}}
 {id=WavePlayer_9, properties {soundWaveAssetPtr=&/IP_GliderWingFlaps_HighBeam_04}}
]
properties {
 attenuationSettings=/Game/Sounds/Attenuation/Fort_Player_Glider_GenericInAir_Close_Athena_Att
 maxDistance=9000
 soundClassObject=&Fort_Gliders_Land
}
